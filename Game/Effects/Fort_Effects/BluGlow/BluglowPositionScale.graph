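# Material function in the BluGlow effects set that supplies normal, position, radius and random values for either particles or objects.
# An isParticle static switch selects particle position and radius, or object position and a radius derived from X object scale.
# Random is a position-based dot product with the RandomizationDistance parameter, while a ParticleRandom/PerInstanceRandom switch is left unconnected.
# Vertex and pixel world normals are output normalized.
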MaterialFunction /Game/Effects/Fort_Effects/BluGlow/BluglowPositionScale
graph {
 properties {description=1/64}
 nodes [
  FunctionOutput(None=Normalize(VectorInput=VertexNormalWS())) {outputName=Vertex Normal}
  FunctionOutput(None=Normalize(VectorInput=PixelNormalWS())) {outputName=Pixel Normal, sortPriority=1}
  n1 = StaticSwitchParameter(True=ParticlePositionWS(), False=ObjectPositionWS()) {parameterName=isParticle}
  FunctionOutput(None=DotProduct(
    A=FunctionOutput(None=@n1) {outputName=Position, sortPriority=4}
    B=ScalarParameter() {defaultValue=-0.0027, parameterName=RandomizationDistance}
   )) {outputName=Random, sortPriority=10}
  n2 = MaterialFunctionCall() {materialFunction=/Engine/Functions/Engine_MaterialFunctions02/WorldPositionOffset/ObjectScale}
  FunctionOutput(None=StaticSwitchParameter(True=ParticleRadius(), False=Multiply(A=@n2.Scale X) {constB=25}) {parameterName=isParticle}) {outputName=Radius, sortPriority=3}
  StaticSwitchParameter(True=ParticleRandom(), False=PerInstanceRandom()) {parameterName=isParticle}
 ]
}
