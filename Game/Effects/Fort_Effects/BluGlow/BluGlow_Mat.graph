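# Subsurface material for glowing blue BluGlow effect meshes, usable as static meshes or mesh particles through static switches.
# Texture-baked morph targets drive vertex offsets and normals via looping time, timestamps or particle age, with optional second morph, random rotation and siphon wiggle.
# Non-particle offsets also pass through a black hole function.
# Emissive layers view-aligned reflection, sparkle, sphere-masked interior glow and rotating noise light shafts, alongside Fresnel-based color and erosion.
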Material /Game/Effects/Fort_Effects/BluGlow/BluGlow_Mat
graph {
 properties {
  bTangentSpaceNormal=false
  bUsedWithMeshParticles=true
  bUsedWithNiagaraMeshParticles=true
  numCustomizedUVs=5
  pixelDepthOffsetMode=PDOM_Legacy
  shadingModel=MSM_Subsurface
  usedShadingModels=Subsurface
 }
 nodes [
  n1 = MaterialFunctionCall() {materialFunction=&/BluglowPositionScale}
  n2 = MaterialFunctionCall() {materialFunction=&BluglowPositionScale}
  Add(
   A=Multiply(
    A=VectorParameter() {parameterName=DiffuseFresnelColor, defaultValue=<rgba 0.0152 0.0927 0.415 1>}.RGB
    B=Fresnel(Normal=@n1.Pixel Normal) {exponent=3}
   )
   B=LinearInterpolate(
    A=VectorParameter() {parameterName=Color Interior, defaultValue=<rgba 0 0.061 0.325 1>}.RGB
    B=VectorParameter() {parameterName=Color, defaultValue=<rgba 0.1159 0.2286 1 1>}.RGB
    Alpha=Fresnel(Normal=@n2.Pixel Normal) {exponent=2}
   )
  )
  n3 = ScalarParameter() {
   defaultValue=1
   desc="Determines the \"roll\" to erode the mesh with. Keep this in a reasonable range (1 to 8)"
   group=Opacity
   parameterName=Erosion Fresnel Brightness
  }
  n4 = ScalarParameter() {
   defaultValue=3
   desc="Determines the \"roll\" to erode the mesh with. Keep this in a reasonable range (1 to 8)"
   group=Opacity
   parameterName=Erosion Fresnel Exponent
  }
  n5 = ParticleColor()
  Clamp(None=Add(
    A=MaterialFunctionCall(Brightness=@n3, Exponent=@n4) {materialFunction=&MF_Fresnel_FX}.Fresnel
    B=ConstantBiasScale(None=@n5.A) {bias=-0.5, scale=2}
   ))
  Constant()
  n6 = ScalarParameter() {defaultValue=480, group=Particle, parameterName=TransitionLength_In_Particle}
  n7 = MaterialFunctionCall(
   TimestampA=ScalarParameter() {group=Particle, parameterName=TimestampA_Particle}
   TransitionLength_In=@n6
   Time=Time()
   Transition In Only?=StaticBoolParameter() {defaultValue=true, group=Particle, parameterName=Transition_In_Only_Particle?}
  ) {materialFunction=&MF_Timestamp_InOut_01}
  n8 = Divide(
   A=LinearInterpolate(
    A=ScalarParameter() {defaultValue=0.5, group=Particle, parameterName=Morph Speed Start}
    B=ScalarParameter() {defaultValue=2, group=Particle, parameterName=Morph Speed End}
    Alpha=@n7.Linear
   )
   B=@n6
  )
  n9 = Power(
   Base=Add(A=@n8, B=Multiply(A=@n8, B=@n6)) {desc=.5002}
   Exp=ScalarParameter() {defaultValue=2, group=Particle, parameterName=Timestamp Curve Exponent}
  )
  n10 = Constant() {r=0}
  n11 = Multiply(A=Time(), B=ScalarParameter() {defaultValue=1, group=Particle, parameterName=Speed Particle})
  n12 = ParticleRelativeTime()
  n13 = TextureObjectParameter() {
   group=Particle
   parameterName=Morph Target Normals A
   samplerType=SAMPLERTYPE_LinearColor
   texture=&/BluGlowAnimation_Normals
  }
  n14 = TextureObjectParameter() {
   group=Particle
   parameterName=Morph Target Offsets A
   samplerType=SAMPLERTYPE_LinearColor
   texture=&/BluGlowAnimation
  }
  n15 = MaterialFunctionCall!(
   Morph Animation=StaticSwitchParameter(
    True=If(A=@n9, B=@n10, "A > B"=@n9, "A == B"=@n11, "A < B"=@n11)
    False=StaticSwitchParameter(True=@n12, False=@n11) {group=Particle, parameterName=UseParticleRelativeTime?}
   ) {group=Particle, parameterName="TimestampAnimation(0-1)Particle"}
   Morph Normal=@n13
   Morph Texture=@n14
   Number of Morph Targets=ComponentMask(None=TextureProperty(Tex=@n13)) {g=true}
   0-1 Animation Value?=StaticBoolParameter() {defaultValue=true, group=Particle, parameterName=Use 0-1 Animation Value?}
   Morph Target UV
  ) {materialFunction=/Engine/Functions/Engine_MaterialFunctions02/MAXScripts/MS_VertexAnimationTools_MorphTargets}
  n16 = LinearInterpolate(
   A=@n15.World Position Offset
   B=Multiply(A=@n15.World Position Offset, B=ScalarParameter() {defaultValue=3, group=Particle, parameterName=WPO Timestamp Scale})
   Alpha=@n7.Linear
  )
  n17 = TextureObjectParameter() {
   group=Particle
   parameterName=Morph Target Normals B
   samplerType=SAMPLERTYPE_LinearColor
   texture=&BluGlowAnimation_Normals
  }
  n18 = TextureObjectParameter() {
   group=Particle
   parameterName=Morph Target Offsets B
   samplerType=SAMPLERTYPE_LinearColor
   texture=&BluGlowAnimation
  }
  n19 = MaterialFunctionCall(
   Morph Animation=ScalarParameter() {defaultValue=1, parameterName=2ndLerpFrame}
   Morph Normal=@n17
   Morph Texture=@n18
   Number of Morph Targets=ComponentMask(None=TextureProperty(Tex=@n17)) {g=true}
   0-1 Animation Value?=StaticBoolParameter() {defaultValue=true, group=Particle, parameterName=Use 0-1 Animation Value?}
  ) {materialFunction=&MS_VertexAnimationTools_MorphTargets}
  n20 = DynamicParameter() {defaultValue=<rgba 1 1 1 0>, paramNames [Param1,Param2,Param3,2ndMorphLERP]}
  n21 = LinearInterpolate(A=@n16, B=@n19.World Position Offset, Alpha=@n20.2ndMorphLERP)
  n22 = Constant()
  n23 = TextureObjectParameter() {parameterName=Morph Target Normals, samplerType=SAMPLERTYPE_LinearColor, texture=&BluGlowAnimation_Normals}
  n24 = MaterialFunctionCall(Float2=TextureProperty(Tex=@n23)) {materialFunction=&BreakOutFloat2Components}
  n25 = Add(
   A=Multiply(A=Time(), B=ScalarParameter() {defaultValue=0.5, parameterName=Speed}) {constB=0.1}
   B=MaterialFunctionCall() {materialFunction=&BluglowPositionScale}.Random
  )
  n26 = Frac(None=@n25)
  n27 = ConstantBiasScale(None=@n26) {bias=-0.5, scale=-2}
  n28 = Power(Base=OneMinus(None=Abs(None=@n27)), Exp=ScalarParameter() {defaultValue=3, parameterName=Animation Power})
  n29 = Multiply(A=@n28) {constB=0.5}
  n30 = MaterialFunctionCall!(Image=OneMinus(None=@n28), Black intensity level=Constant() {r=0.5}) {materialFunction=/Engine/Functions/Engine_MaterialFunctions02/Texturing/RaiseBlackLevelsByPercentage}
  n31 = Floor(None=Multiply(A=@n26) {constB=2})
  n32 = Multiply(A=Divide(B=Subtract(A=@n24.G)) {constB=4}, B=Add(A=Floor(None=@n25), B=LinearInterpolate(A=@n29, B=@n30.Result, Alpha=@n31)))
  n33 = TextureObjectParameter() {parameterName=Morph Target Offsets, samplerType=SAMPLERTYPE_LinearColor, texture=&BluGlowAnimation}
  n34 = MaterialFunctionCall(
   Morph Animation=StaticSwitchParameter(True=MaterialFunctionCall() {materialFunction=&MF_Timestamp_InOut_01}.Linear, False=@n32) {group=Timestamp, parameterName="TimestampAnimation(0-1)"}
   Morph Normal=@n23
   Morph Texture=@n33
   Number of Morph Targets=@n24.G
   0-1 Animation Value?=StaticBool() {value=true}
  ) {materialFunction=&MS_VertexAnimationTools_MorphTargets}
  n35 = TextureSampleParameter2D(UVs=Panner(Coordinate=TextureCoordinate() {uTiling=0.6439, vTiling=0.5879}) {speedX=0.1596, speedY=-0.2278}) {group=Wiggle, parameterName=Wiggle Normal, samplerType=SAMPLERTYPE_Normal, texture=/Game/Global/Textures/Default/Noise/T_SwirlNoise_N}
  n36 = TextureSampleParameter2D(UVs=Panner(Coordinate=TextureCoordinate() {uTiling=0.5103, vTiling=0.4787}) {speedX=-0.1806, speedY=0.2604}) {group=Wiggle, parameterName=Wiggle Normal, samplerType=SAMPLERTYPE_Normal, texture=&T_SwirlNoise_N}
  n37 = VectorParameter() {group=Wiggle, parameterName=Wiggle Intensity, defaultValue=<rgba 25 25 25 0>}
  n38 = Multiply(
   A=DotProduct(A=AppendVector(A=@n35.R, B=@n35.R), B=AppendVector(A=@n36.R, B=@n36.R))
   B=Multiply(A=VertexNormalWS(), B=Transform(None=@n37.RGB) {transformType=TRANSFORM_Local})
  )
  n39 = VertexColor()
  n40 = StaticSwitchParameter(True=Add(A=@n34.World Position Offset, B=Multiply(A=@n38, B=@n39.R)), False=@n34.World Position Offset) {group=Wiggle, parameterName=SiphonWiggle}
  n41 = MaterialFunctionCall() {materialFunction=&BluglowPositionScale}
  n42 = Add(A=@n41.Random, B=Multiply(A=Add(A=@n25, B=@n32)) {constB=0.2})
  n43 = MaterialFunctionCall(
   Hue Shift Percentage=@n42
   Texture=MaterialFunctionCall(Hue Shift Percentage=@n42, Texture=Constant3Vector() {constant=<rgba 1 0 0 0>}) {materialFunction=&HueShift}.Result
  ) {materialFunction=&HueShift}
  n44 = RotateAboutAxis(NormalizedRotationAxis=@n43.Result, RotationAngle=@n42, PivotPoint=@n41.Position, Position=Add(A=@n40, B=WorldPosition().XYZ))
  n45 = StaticSwitchParameter(
   True=StaticSwitchParameter(True=StaticSwitchParameter(True=@n21, False=@n16) {group=Particle, parameterName=ParticleUse2ndMoprh?}, False=@n22) {parameterName=UseParticleMorphTargets}
   False=StaticSwitchParameter(True=Add(A=@n40, B=@n44), False=@n40) {defaultValue=true, parameterName=UseRandomRotate?}
  ) {group=Particle, parameterName=isParticle}
  n46 = Add(A=@n45, B=WorldPosition().XYZ)
  MaterialExpressionStaticSwitchParameter_3 = StaticSwitchParameter(
   True=@n45
   False=Add(
    A=@n45
    B=MaterialFunctionCall!(WorldPostionOfBlackHole, World Position=@n46) {materialFunction=/Game/Packages/Fortnite_UniqueMaterials_1/BlackHoleFunction}.Result
   )
  ) {group=Particle, parameterName=isParticle}
  Custom!(WorldPosition=@MaterialExpressionStaticSwitchParameter_3) {
   code="##if XBOXONE_PROFILE\r\n##define XBOXONE_BIAS_HACK 1\r\n##endif\r\nreturn WorldPosition;"
   inputs [{inputName=WorldPosition, input {expression=&self:MaterialExpressionStaticSwitchParameter_3}}]
  }
  n47 = TextureSample!(
   UVs=MaterialFunctionCall!(ReflectionVector, ReflectionTexture) {materialFunction=/Engine/Functions/Engine_MaterialFunctions01/Reflections/ViewAlignedReflection}.UVs
   Tex
   Bias=ScalarParameter() {parameterName=Reflection Mip Bias}
   Apply View MipBias
  ) {mipValueMode=TMVM_MipBias, texture=/Game/Packages/Fortress_SharedMaterials/Textures/T_Ravens_SpecularReflection}
  n48 = Multiply(
   A=ScalarParameter() {defaultValue=4, parameterName=Inverted Fresnel Brightness}
   B=MaterialFunctionCall(
    Normal Vector=MaterialFunctionCall() {materialFunction=&BluglowPositionScale}.Pixel Normal
    Invert Fresnel=StaticBool() {value=true}
    Power=ScalarParameter() {defaultValue=2, parameterName=Inverted Fresnel Power}
   ) {materialFunction=&Fresnel_Function}.Result
  )
  n49 = MaterialFunctionCall!(
   Scale=ScalarParameter() {defaultValue=-0.1, parameterName=SparkleScale}
   Sparkle World center point=MaterialFunctionCall() {materialFunction=&BluglowPositionScale}.Position
  ) {materialFunction=/Game/Packages/Fortnite_UniqueMaterials_1/Effects/FortniteSparkle}
  n50 = Multiply(A=@n49.Result, B=ScalarParameter() {defaultValue=0.211, parameterName=Sparkle brightness})
  n51 = Constant()
  n52 = Transform(None=PixelNormalWS()) {transformSourceType=TRANSFORMSOURCE_World, transformType=TRANSFORM_View}
  n53 = Add(
   A=@n49.UVS
   B=Multiply(A=ComponentMask(None=@n52) {g=true, r=true}, B=ScalarParameter() {parameterName=Interior Glow Distortion}) {constB=0.1}
  )
  n54 = ScalarParameter() {defaultValue=1.0326, parameterName=InternalGlowRadius}
  n55 = StaticSwitchParameter(
   True=Multiply(A=@n54, B=ParticleRadius())
   False=Multiply(A=MaterialFunctionCall() {materialFunction=&BluglowPositionScale}.Radius, B=@n54)
  ) {group=Particle, parameterName=isParticle}
  n56 = Transform(None=PixelNormalWS()) {transformSourceType=TRANSFORMSOURCE_World, transformType=TRANSFORM_View}
  n57 = Multiply(
   A=Add(
    A=MaterialFunctionCall(Float3=@n56) {materialFunction=&BreakOutFloat3Components}.G
    B=ScalarParameter() {defaultValue=-0.5, parameterName=Vertical Darkening Offset}
   )
   B=ScalarParameter() {defaultValue=-0.25, parameterName=Vertical Darkening Amount}
  )
  n58 = Multiply(A=OneMinus(None=SphereMask(A=@n51, B=@n53, Radius=@n55) {attenuationRadius=0.5, hardnessPercent=0}), B=OneMinus(None=Clamp(None=@n57)))
  n59 = Clamp(None=@n58)
  n60 = ScalarParameter() {defaultValue=4, parameterName=GlowAmount}
  n61 = Multiply(A=Add(A=Multiply(A=@n47.RGB, B=@n48), B=Add(B=@n50) {constA=0.5}), B=Add(A=Multiply(A=@n59, B=@n60), B=@n58) {constB=0.1})
  n62 = VectorParameter() {parameterName=2D Glow color, defaultValue=<rgba 0.1537 0.2538 0.53 2>}
  n63 = Divide(
   A=@n53
   B=Multiply(
    A=MaterialFunctionCall() {materialFunction=&BluglowPositionScale}.Radius
    B=ScalarParameter() {defaultValue=2.7632, parameterName=Scale On Center Light Shafts}
   ) {constB=2}
  )
  n64 = Add(A=@n63) {constB=0.5}
  n65 = Add(A=MaterialFunctionCall() {materialFunction=&BluglowPositionScale}.Random, B=Time())
  n66 = Multiply(A=@n63) {constB=1.5}
  n67 = TextureSample!(UVs=Rotator(Coordinate=Add(A=@n66) {constB=0.5}, Time=@n65) {speed=-0.1}, Tex, Apply View MipBias) {texture=&/ZoomNoise}
  n68 = Multiply(A=TextureSample(UVs=Rotator(Coordinate=@n64, Time=@n65)) {texture=&ZoomNoise}.R, B=Add(A=@n67.G, B=@n67.G))
  n69 = LinearInterpolate(
   A=VectorParameter() {parameterName=Light shaft Color 1, defaultValue=<rgba 0.4121 0.4795 0.87 1>}.RGB
   B=VectorParameter() {parameterName=Light shaft Color 2, defaultValue=<rgba 0.8597 0.905 0.8911 1>}.RGB
   Alpha=@n68
  )
  n70 = Multiply(A=@n68, B=ScalarParameter() {defaultValue=1, desc=For Blueprints, parameterName=LightShaftMultipler}) {constB=2.233}
  n71 = Power(
   Base=MaterialFunctionCall!(Base=Multiply(A=@n61, B=@n62.RGB), Add=Multiply(A=@n69, B=@n70)) {materialFunction=/Engine/Functions/Engine_MaterialFunctions02/Math/MultiplyAdd}.Result
   Exp=@n62.A
  )
  n72 = VectorParameter() {parameterName=Glow Pass, defaultValue=<rgba 0 0.0314 0.27 0.1>}
  MaterialFunctionCall(Emissive=Multiply(A=Add(A=@n71, B=@n72.RGB), B=ParticleColor().A)) {materialFunction=&MF_FX_EmissiveAdjustments}
  Multiply(
   A=VectorParameter() {parameterName=SubSurfaceColor, defaultValue=<rgba 0.0067 0.0103 0.02 1>}.RGB
   B=MaterialFunctionCall() {materialFunction=/Game/Packages/Fortress_SharedMaterials/Base_Material_Functions/CascadingShadowMapMask}.Cascading Shadow Mask
  )
  n73 = PixelNormalWS()
  n74 = CameraVectorWS()
  n75 = OneMinus(None=Clamp(None=DotProduct(A=@n73, B=@n74)))
  n76 = ScalarParameter() {defaultValue=0.7493, parameterName=SubSurfaceFresnelExponent}
  OneMinus(None=Multiply(A=Power(Base=@n75, Exp=@n76), B=ScalarParameter() {defaultValue=0.1035, parameterName=SubSurfaceOpacity}))
  Reroute(None=StaticSwitchParameter(True=Constant(), False=@n32) {parameterName=IsParticle})
  ScalarParameter() {defaultValue=0.1101, parameterName=Roughness}
  ScalarParameter() {defaultValue=-6.8985, parameterName=Metallic}
  ScalarParameter() {defaultValue=0.1514, parameterName=Specular}
  n77 = LinearInterpolate(A=@n15."Custom UV 2 (Needed)", B=@n19."Custom UV 2 (Needed)", Alpha=@n20.2ndMorphLERP)
  StaticSwitchParameter(
   True=StaticSwitchParameter(True=StaticSwitchParameter(True=@n77, False=@n15."Custom UV 2 (Needed)") {group=Particle, parameterName=ParticleUse2ndMoprh?}, False=@n22) {parameterName=UseParticleMorphTargets}
   False=@n34."Custom UV 2 (Needed)"
  ) {group=Particle, parameterName=isParticle}
  n78 = LinearInterpolate(A=@n15."Custom UV 3 (Needed)", B=@n19."Custom UV 3 (Needed)", Alpha=@n20.2ndMorphLERP)
  StaticSwitchParameter(
   True=StaticSwitchParameter(True=StaticSwitchParameter(True=@n78, False=@n15."Custom UV 3 (Needed)") {group=Particle, parameterName=ParticleUse2ndMoprh?}, False=@n22) {parameterName=UseParticleMorphTargets}
   False=@n34."Custom UV 3 (Needed)"
  ) {group=Particle, parameterName=isParticle}
  n79 = LinearInterpolate(A=@n15."Pixel Shader World Vertex Normal (See tooltip)", B=@n19."Pixel Shader World Vertex Normal (See tooltip)", Alpha=@n20.2ndMorphLERP)
  n80 = Normalize(VectorInput=Add(A=@n35.RGB, B=@n36.RGB))
  n81 = StaticSwitchParameter(
   True=LinearInterpolate(A=@n34."Pixel Shader World Vertex Normal (See tooltip)", B=Transform(None=@n80), Alpha=@n39.R)
   False=@n34."Pixel Shader World Vertex Normal (See tooltip)"
  ) {group=Wiggle, parameterName=SiphonWiggle}
  StaticSwitchParameter(
   True=StaticSwitchParameter(
    True=StaticSwitchParameter(True=@n79, False=@n15."Pixel Shader World Vertex Normal (See tooltip)") {group=Particle, parameterName=ParticleUse2ndMoprh?}
    False=VertexNormalWS()
   ) {parameterName=UseParticleMorphTargets}
   False=StaticSwitchParameter(
    True=MaterialFunctionCall(Rotation Axis=@n43.Result, Rotation Angle=@n42, World Space Vertex Normals=@n81) {materialFunction=&FixRotateAboutAxisNormals}.New World Space Vertex Normal
    False=@n81
   ) {defaultValue=true, parameterName=UseRandomRotate?}
  ) {group=Particle, parameterName=isParticle}
  VectorParameter() {parameterName=ClearCoatColor, defaultValue=<rgba 0.0845 0.2256 0.5 -1.9618>}
 ]
}
