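# Additive, unlit, two-sided particle material for glowing blue rays, named for the BluGlo bucket collect effect and usable on mesh and Niagara mesh particles.
# Two panned plasma noise samples, sped by a per-particle Time Random parameter, are summed and partly rounded for harder edges.
# That noise plus a V-based fade blends outer and inner colors, then is shaped by Fresnel and particle color.
# Opacity uses depth fade, skipped on mobile.
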Material /Game/Effects/Fort_Effects/BluGlow/M_BluGlo_BucketCollectRays_01
graph {
 properties {
  bUsedWithMeshParticles=true
  bUsedWithNiagaraMeshParticles=true
  blendMode=BLEND_Additive
  pixelDepthOffsetMode=PDOM_Legacy
  shadingModel=MSM_Unlit
  twoSided=true
  usedShadingModels=Unlit
 }
 nodes [
  n1 = ParticleColor()
  n2 = TextureCoordinate()
  n3 = Multiply(
   A=AppendVector(A=ScalarParameter() {defaultValue=2, parameterName=U Repeat A}, B=ScalarParameter() {defaultValue=0.0196, parameterName=V Repeat A})
   B=@n2
  )
  n4 = Multiply(A=Time(), B=DynamicParameter() {paramNames [Time Random, Param2, Param3, Param4]}.Time Random)
  n5 = Multiply(
   A=@n2
   B=AppendVector(A=ScalarParameter() {defaultValue=1, parameterName=U Repeat B}, B=ScalarParameter() {defaultValue=0.0117, parameterName=V Repeat B})
  )
  n6 = Add(
   A=TextureSample(UVs=Panner(Coordinate=@n3, Time=@n4) {bFractionalPart=true, speedX=0.031, speedY=0.027}) {texture=/Game/Effects/Fort_Effects/Textures/Plasma/T_PlasmaNoise_A_01}.B
   B=TextureSample(UVs=Panner(Coordinate=@n5, Time=@n4) {bFractionalPart=true, speedX=-0.1485, speedY=0.0218}) {texture=&T_PlasmaNoise_A_01}.R
  )
  n7 = LinearInterpolate(
   A=MaterialFunctionCall!(In=@n6) {materialFunction=/Engine/Functions/Engine_MaterialFunctions02/Utility/Round}.Result
   B=@n6
   Alpha=ScalarParameter() {defaultValue=0.6, parameterName=Edge LERP}
  )
  n8 = ComponentMask(None=TextureCoordinate()) {g=true}
  n9 = ScalarParameter() {defaultValue=0.17, parameterName=Bottom Fade Bias}
  n10 = Power(Base=Clamp(None=Add(A=@n8, B=@n9)), Exp=ScalarParameter() {defaultValue=7, parameterName=Bottom Fade Exponent}) {constExponent=20}
  n11 = Multiply(B=@n10) {constA=5}
  n12 = Power(
   Base=LinearInterpolate(
    A=VectorParameter() {parameterName=Color Outer, defaultValue=<rgba 0 0.1634 0.87 1>}.RGB
    B=VectorParameter() {parameterName=Color Inner, defaultValue=<rgba 0.22 0.8912 2 0>}.RGB
    Alpha=Add(A=@n7, B=@n11)
   )
   Exp=LinearInterpolate(Alpha=@n8) {constA=2}
  )
  n13 = MaterialFunctionCall(
   Brightness=ScalarParameter() {defaultValue=1, parameterName=Fresnel Brightness}
   Exponent=ScalarParameter() {defaultValue=2, parameterName=Fresnel Exponent}
  ) {materialFunction=&MF_Fresnel_FX}
  n14 = ScalarParameter() {parameterName=UseDoubleSideFresnel}
  n15 = Add(A=Multiply(A=LinearInterpolate(A=@n13.Fresnel, B=@n13.Fresnel, Alpha=@n14), B=@n8) {constA=1}, B=@n10)
  MaterialFunctionCall(Emissive=Multiply(A=@n1.RGB, B=Multiply(A=@n12, B=@n15))) {materialFunction=&MF_FX_EmissiveAdjustments}
  n16 = Multiply(A=@n1.A, B=@n15)
  n17 = DepthFade(Opacity=@n16, FadeDistance=ScalarParameter() {group=Opacity, parameterName=Depth Fade})
  ShadingPathSwitch(Default=@n17, Deferred=@n17, Mobile=@n16) {desc=Migrated from FeatureLevelSwitch}
 ]
}
