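# M_Bluglo_Refraction_02 is an unlit translucent refraction material for BluGlow sprite particles in Cascade and Niagara.
# A dynamic parameter supplies per-particle index of refraction, blended from a neutral 1 through DepthFade.
# The "Refraction Intersection Distance" parameter sets that fade, softening distortion where sprites meet scene geometry.
# A spherical normal map borrowed from the Teleporter gadget is sampled, presumably to shape the distortion.
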Material /Game/Effects/Fort_Effects/BluGlow/M_Bluglo_Refraction_02
graph {
 properties {
  bAutomaticallySetUsageInEditor=false
  bUsedWithNiagaraSprites=true
  bUsedWithParticleSprites=true
  bUsedWithStaticMesh=false
  blendMode=BLEND_Translucent
  pixelDepthOffsetMode=PDOM_Legacy
  refractionMethod=RM_IndexOfRefraction
  shadingModel=MSM_Unlit
  usedShadingModels=Unlit
 }
 nodes [
  Constant()
  LinearInterpolate(
   B=DynamicParameter() {paramNames [Param1,Param2,Param3,Refraction]}.Refraction
   Alpha=DepthFade(FadeDistance=ScalarParameter() {defaultValue=120, parameterName=Refraction Intersection Distance})
  ) {constA=1}
  TextureSample() {samplerType=SAMPLERTYPE_Normal, texture=/Game/Gadgets/Assets/Teleporter/Textures/T_Spherical_Normal_02}
 ]
}
